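; Creative object interaction mode Blueprint for the phone tool's actor-targeting mode in the ObjectMover weapon.
; On BeginPlay it runs parent logic, then binds the delete-actors interaction behavior.
; It constructs target detector and evaluator objects from soft class references on the mode.
; It forwards device prop placement, including the preview flag, to the Creative device prop parent and drives an added-to-quickbar material pulse through a timeline.
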(event EventBeginPlay
  (|CallParentFunction)
  (CallFunction|EventAssignDeleteInteractBehavior))

(event Custom|EventAssignDeleteInteractBehavior
  ;[0.8385 0.8385 0.8385 0.5] Assign Delete ObjectInteractionBehavior
  (Variables|ObjectInteractionMode|SetBoundDeleteInteractBehavior (Variables|Default|GetCreativeInteraction_DeleteActors)))

(event EventOnCreativeDevicePropPlaced
  (params (DeviceProp :type Fort%20Creative%20Device%20Prop%20Object%20Reference) (bForPreview :type Boolean))
  (bind _ascreative_device_prop_parent (Utilities|Casting|CastToCreative_Device_Prop_Parent DeviceProp))
  (Class|CreativeDevicePropParent|OnDevicePlaced _ascreative_device_prop_parent bForPreview))

(event AddedtoQuickbarFX|EventPlayAddedToQuickbarAnimation
  ;[0.8385 0.8385 0.8385 0.5] Playing Added to Quickbar Animation on target materials
  (bind (_direction _lerp) (Timeline|QuickbarPulseTL PlayFromStart)
    (:Update
      (AddedtoQuickbarFX|UpdateAddedtoQuickbarAnimation _lerp))))

(fn InitializeInteractionTargetDetector
  (bind _returnvalue (Game|ConstructObjectfromClass (Utilities|ResolveSoftReference (Variables|ObjectInteractionMode|GetInteractionTargetDetectionMethod))))
  (return _returnvalue))

(fn InitializeInteractionTargetEvaluator
  (bind _returnvalue (Game|ConstructObjectfromClass (Utilities|ResolveSoftReference (Variables|ObjectInteractionMode|GetInteractionTargetEvaluatorMethod))))
  (return _returnvalue))

(fn ConstructionScript)
