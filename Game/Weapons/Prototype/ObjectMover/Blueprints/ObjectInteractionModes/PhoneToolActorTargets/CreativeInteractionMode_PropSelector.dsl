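; Prop selector interaction mode for the prototype Object Mover weapon, filed under phone tool actor targets in creative interaction modes.
; It overrides two initializers that create the interaction target detector and the target evaluator.
; Each resolves a soft class reference from the mode's detection or evaluator method variable, then constructs and returns an object of that class.
; The construction script is empty, and no selection logic appears in this graph.
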(fn InitializeInteractionTargetDetector
  (bind _returnvalue (Game|ConstructObjectfromClass (Utilities|ResolveSoftReference (Variables|ObjectInteractionMode|GetInteractionTargetDetectionMethod))))
  (return _returnvalue))

(fn InitializeInteractionTargetEvaluator
  (bind _returnvalue (Game|ConstructObjectfromClass (Utilities|ResolveSoftReference (Variables|ObjectInteractionMode|GetInteractionTargetEvaluatorMethod))))
  (return _returnvalue))

(fn ConstructionScript)
